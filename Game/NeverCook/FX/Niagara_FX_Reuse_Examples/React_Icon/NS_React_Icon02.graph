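# Niagara reaction-icon effect from the reuse examples, with one sprite emitter, NE_Laugh, that bursts a short-lived icon from a zero-radius sphere location.
# A curve grows the sprite from zero to full size and back to zero over its life.
# Scale Color picks random RGB between two bright color-curve gradients, and dynamic material parameters feed the sprite material.
# Front-end and preview user booleans are exposed.
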NiagaraSystem /Game/NeverCook/FX/Niagara_FX_Reuse_Examples/React_Icon/NS_React_Icon02
dependencies {
 usedDataInterfaces [&NiagaraDataInterfaceCurve, &NiagaraDataInterfaceColorCurve]
 usedDynamicInputs [&FN_FrontEndFloatModifier]
 usedModules [&SystemState]
}
emitters {NE_Laugh {
  emitterUpdateScript [
   &EmitterState(Life Cycle Mode=::NewEnumerator1, Loop Behavior=::NewEnumerator1, Loop Duration=0.2, UseLoopDelay=false)
   &SpawnBurst_Instantaneous(Use Spawn Probability=false, UseLoopCountLimit=false)
  ]
  particleSpawnScript [
   &InitializeParticle(
    Lifetime Max=1
    Lifetime Mode=::NewEnumerator1
    Sprite Size Mode=::NewEnumerator3
    Uniform Sprite Size=20
    UsePositionOffset=false
    Write Material Random=true
   ) {module=InitializeParticle001}
   &ApplyInitialForces()
   &ShapeLocation(MaskSpawnGroup=false, Override Seed=false, Sphere Radius=0, UseOwnerScale=true)
  ]
  particleUpdateScript [
   &ParticleState()
   &ScaleSpriteSize(Uniform Curve Sprite Scale=DataInterface() {propertyValues {Curve {keys [
        {interpMode=RCIM_Cubic, tangentMode=RCTM_User, time=0, value=0}
        {interpMode=RCIM_Cubic, tangentMode=RCTM_User, time=0.5, value=1}
        {tangentMode=RCTM_User, time=1, value=0}
       ]}}})
   &Drag(Drag=&RandomRangeFloat(Minimum=8, Maximum=9), Use Linear Drag=true, Use Rotational Drag=false)
   &SolveForcesAndVelocity(Clamp Velocity=false, Limit Acceleration=false)
   &ScaleColor(
    Scale RGB=&RandomRangeVector(
     Minimum=&MakeVectorFromLinearColorRGB(Color=&ColorFromCurve(ColorCurve=DataInterface() {propertyValues {
         AlphaCurve {keys [<at 0 1 RCIM_Cubic>]}
         BlueCurve {keys [<at 0.1976 0.0986 RCIM_Cubic>,<at 0.4904 0>,<at 0.7949 2.5 RCIM_Cubic>]}
         GreenCurve {keys [<at 0.1976 0 RCIM_Cubic>,<at 0.4904 0.7938>,<at 0.7949 0 RCIM_Cubic>]}
         RedCurve {keys [<at 0.1976 2.5 RCIM_Cubic>,<at 0.4904 2.5>,<at 0.7949 1.5065 RCIM_Cubic>]}
        }}))
     Maximum=&MakeVectorFromLinearColorRGB(Color=&ColorFromCurve(ColorCurve=DataInterface() {propertyValues {
         AlphaCurve {keys [<at 0 1 RCIM_Cubic>]}
         BlueCurve {keys [<at 0.1976 0 RCIM_Cubic>,<at 0.4904 2.4045>,<at 0.7949 2.5 RCIM_Cubic>]}
         GreenCurve {keys [<at 0.1976 2.5 RCIM_Cubic>,<at 0.4904 2.5>,<at 0.7949 0.2813 RCIM_Cubic>]}
         RedCurve {keys [<at 0.1976 0.8261 RCIM_Cubic>,<at 0.4904 0.0729>,<at 0.7949 0.1077 RCIM_Cubic>]}
        }}))
    )
    ScaleA=true
    ScaleRGB=true
   )
   &DynamicMaterialParameters(
    Index 0 Param 1=1
    Index 0 Param 2=1
    Index 0 Param 3=1
    Index 0 Param 4=1
    Module.Index0Float0Write=true
    Module.Index0Float1Write=true
    Module.Index0Float2Write=true
    Module.Index0Float3Write=true
   )
  ]
  rendererClasses [&NiagaraSpriteRendererProperties]
  renderers [{rendererClass=&NiagaraSpriteRendererProperties}]
 }}
summary {
 systemName=NS_React_Icon02
 emitters [{emitterName=NE_Laugh, rendererClasses [&NiagaraSpriteRendererProperties]}]
 userVariables [
  User.bIsFrontEnd {type=NiagaraBool}
  User.bIsFrontEndPreview {type=NiagaraBool}
 ]
}
